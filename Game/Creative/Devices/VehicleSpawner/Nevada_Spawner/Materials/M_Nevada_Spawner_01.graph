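# Lit surface material for the Nevada vehicle spawner device in Creative, with a glowing, tinted logo.
# Base color, normal and mask textures feed material attributes, with the mask's green channel shared by metallic, specular and roughness.
# A second-UV logo mask blends its red and intensity-scaled green channels to drive emissive, then MF_GlowFlicker processes the attributes.
# Logo UV scale/offset and Fresnel or bounds-based glow branches are left unconnected.
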Material /Game/Creative/Devices/VehicleSpawner/Nevada_Spawner/Materials/M_Nevada_Spawner_01
graph {
 properties {bUseMaterialAttributes=true, pixelDepthOffsetMode=PDOM_Legacy, usedShadingModels=Default Lit}
 nodes [
  n1 = TextureCoordinate() {coordinateIndex=1, desc=Second UV Set}
  Add(
   A=MaterialFunctionCall(UVs=@n1, Texture Scale=ScalarParameter() {defaultValue=1, group=Logo, parameterName=Logo Scale}) {materialFunction=&ScaleUVsByCenter}.UVs
   B=AppendVector(A=ScalarParameter() {group=Logo, parameterName=Logo Offset U}, B=ScalarParameter() {group=Logo, parameterName=Logo Offset V})
  )
  n2 = TextureSampleParameter2D() {
   parameterName=Specular
   samplerType=SAMPLERTYPE_Masks
   texture=/Game/Creative/Devices/VehicleSpawner/Nevada_Spawner/Textures/T_Device_Nevada_Spawner_01_S
  }
  n3 = VectorParameter() {group=Logo, parameterName=Emissive, sortPriority=1, defaultValue=<rgba 0 1 0.9387 1>}
  n4 = TextureSampleParameter2D(UVs=@n1) {
   group=Logo
   parameterName=Logo Mask
   samplerSource=SSM_Clamp_WorldGroupSettings
   samplerType=SAMPLERTYPE_LinearColor
   texture=/Game/Creative/Devices/VehicleSpawner/Nevada_Spawner/Textures/T_Device_Nevada_Spawner_01_E
  }
  n5 = LinearInterpolate(A=Multiply(A=@n4.G, B=ScalarParameter() {defaultValue=0.2, parameterName=Glow Intensity}), B=@n4.R, Alpha=@n4.R) {constAlpha=1}
  n6 = TextureSampleParameter2D() {
   parameterName=Normal
   samplerType=SAMPLERTYPE_Normal
   texture=/Game/Creative/Devices/VehicleSpawner/Nevada_Spawner/Textures/T_Device_Nevada_Spawner_01_N
  }
  MaterialFunctionCall!(MA=MakeMaterialAttributes!(
    BaseColor=TextureSampleParameter2D() {parameterName=Basecolor, texture=/Game/Creative/Devices/VehicleSpawner/Nevada_Spawner/Textures/T_Device_Nevada_Spawner_01_D}.RGB
    Metallic=@n2.G
    Specular=@n2.G
    Roughness=@n2.G
    Anisotropy
    EmissiveColor=LinearInterpolate(B=@n3.RGB, Alpha=@n5) {constAlpha=1}
    Opacity
    OpacityMask
    Normal=@n6.RGB
    Tangent
    WorldPositionOffset
    SubsurfaceColor
    ClearCoat
    ClearCoatRoughness
    AmbientOcclusion
    Refraction
    CustomizedUVs_0
    CustomizedUVs_1
    CustomizedUVs_2
    CustomizedUVs_3
    CustomizedUVs_4
    CustomizedUVs_5
    CustomizedUVs_6
    CustomizedUVs_7
    PixelDepthOffset
    ShadingModel
    Displacement
   )) {materialFunction=/Game/Environments/Sets/PrincessCastle/Props/Materials/MF_GlowFlicker}
  n7 = MaterialFunctionCall(
   In=Fresnel(
    ExponentIn=ScalarParameter() {defaultValue=0.2158, parameterName=ExponentIn}
    BaseReflectFractionIn=ScalarParameter() {parameterName=BaseReflectFractionIn}
   )
   Contrast=ScalarParameter() {defaultValue=0.2713, parameterName="Contrast (S)"}
  ) {materialFunction=&CheapContrast}
  Multiply(A=@n7.Result)
  n8 = VectorParameter() {parameterName=Color_Outer, defaultValue=<rgba 0 0.0304 0.3229 1>}
  n9 = VectorParameter() {parameterName=Color_Inner_Core, defaultValue=<rgba 0 2.6551 3 1>}
  n10 = PreSkinnedLocalBounds()
  n11 = Subtract(A=VertexInterpolator(VS=PreSkinnedPosition()).PS, B=@n10.Min)
  n12 = Subtract(A=@n10.Max, B=@n10.Max)
  n13 = Frac(None=Multiply(A=Divide(A=@n11, B=@n12), B=VectorParameter() {parameterName=Tile_Rate, defaultValue=<rgba 1 1 3 1>}.RGB) {constB=3})
  n14 = ComponentMask(None=@n13) {b=true}
  n15 = OneMinus(None=@n14)
  n16 = Divide(
   A=Add(A=ScalarParameter() {defaultValue=-0.0286, parameterName=Z_offset}, B=Multiply(A=@n15, B=@n14))
   B=ScalarParameter() {defaultValue=0.2308, parameterName=Z_Falloff}
  )
  n17 = ComponentMask(None=@n13) {g=true}
  n18 = OneMinus(None=@n17)
  n19 = Divide(
   A=Add(A=Multiply(A=@n18, B=@n17), B=ScalarParameter() {defaultValue=-0.0381, parameterName=Y_offset})
   B=ScalarParameter() {defaultValue=0.0762, parameterName=Y_Falloff}
  )
  n20 = Power(Base=Multiply(A=Saturate(None=@n16), B=Saturate(None=@n19)), Exp=ScalarParameter() {defaultValue=2.023, parameterName=Thickness})
  n21 = Abs(None=DotProduct(A=CameraVectorWS(), B=VertexNormalWS()))
  n22 = ScalarParameter() {defaultValue=8.8301, parameterName=Fresnel _Falloff}
  n23 = Add(A=Saturate(None=Power(Base=@n21, Exp=@n22)), B=ScalarParameter() {defaultValue=0.25, parameterName=Fresnel _Strength})
  n24 = LinearInterpolate(A=@n8.RGB, B=@n9.RGB, Alpha=Multiply(A=Saturate(None=@n20), B=Saturate(None=@n23)))
  Multiply(A=@n24)
  n25 = LinearInterpolate(A=@n9.RGB, B=@n8.RGB, Alpha=@n7.Result)
  Multiply(
   A=Multiply(
    A=StaticSwitchParameter(True=@n24, False=@n25) {parameterName=Use_Bounds_for uvs}
    B=ScalarParameter() {defaultValue=10, parameterName=Brightness multplyer}
   )
   B=TextureSample(UVs=TextureCoordinate() {coordinateIndex=1}) {texture=&T_Device_Nevada_Spawner_01_E}.RGB
  )
  TextureCoordinate()
 ]
}
